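; Base Reality Flower building actor that, on the server at BeginPlay, spawns rings of bounce pads at rising heights with a ring of seed pods above them.
; Scalable float rows build one spawner params struct per group, each holding a pre-sized array for its spawned actors.
; SpawnBGA spreads actors evenly by angle with random angular, radius, height and yaw variation, killing any valid occupant of the slot before storing the new one.
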(event EventBeginPlay
  ;[1 1 1] Begin Play
  (|CallParentFunction)
  (Networking|SwitchHasAuthority
    (:Authority
      (CallFunction|InitializeSpawnerParams)
      (CallFunction|SpawnInitialBGAs))))

(event Building|EventOnBuildingActorInitialized
  (params (InitializationReason :type EFortBuildingInitializationReason%20Enum :default "StaticallyPlaced") (BuildingPersistentState :type EFortBuildingPersistentState%20Enum :default "Default"))
  ;[1 1 1] Building Initialized
  (|CallParentFunction InitializationReason BuildingPersistentState))

(event Custom|InitializeSpawnerParams
  (bind _row_seed_count (Variables|Default|GetRow_Seed_Count))
  (bind _returnvalue (ScalableFloat|GetValueatLevel _row_seed_count))
  (bind _returnvalue_2 (Math|Float|Truncate _returnvalue))
  (bind _row_bouncepads_countpergroup (Variables|Default|GetRow_BouncePads_CountPerGroup))
  (bind _returnvalue_3 (ScalableFloat|GetValueatLevel _row_bouncepads_countpergroup))
  (bind _returnvalue_4 (Math|Float|Truncate _returnvalue_3))
  ;[1 1 1] Initialize Spawner Params
  (Utilities|Array|Clear (Variables|Default|GetSpawnerParams))
  (for _index (range (Math|Float|Truncate (ScalableFloat|GetValueatLevel (Variables|Default|GetRow_BouncePads_GroupsToSpawn))))
    (CallFunction|CreateNewSizedNullBGAArray _returnvalue_4)
    (Utilities|Array|Add (Variables|Default|GetSpawnerParams) (Utilities|Struct|MakeRealityFlowerSpawnerParams _index (Variables|Default|GetBGA_BouncePad) _returnvalue_4 (ScalableFloat|GetValueatLevel (Variables|Default|GetRow_BouncePads_RadiusMin)) (ScalableFloat|GetValueatLevel (Variables|Default|GetRow_BouncePads_RadiusMax)) (ScalableFloat|GetValueatLevel (Variables|Default|GetRow_BouncePads_RadialSpacingVarianceMult)) (+ (* _index (ScalableFloat|GetValueatLevel (Variables|Default|GetRow_BouncePads_AdditionalHeightPerGroup))) (ScalableFloat|GetValueatLevel (Variables|Default|GetRow_BouncePads_StartingHeight))) (+ (+ (* _index (ScalableFloat|GetValueatLevel (Variables|Default|GetRow_BouncePads_AdditionalHeightPerGroup))) (ScalableFloat|GetValueatLevel (Variables|Default|GetRow_BouncePads_StartingHeight))) (ScalableFloat|GetValueatLevel (Variables|Default|GetRow_BouncePads_HeightVariance))) (Variables|Default|GetSizedBGAsNullArray))))
  (CallFunction|CreateNewSizedNullBGAArray _returnvalue_2)
  (Utilities|Array|Add (Variables|Default|GetSpawnerParams) (Utilities|Struct|MakeRealityFlowerSpawnerParams (Math|Float|Truncate (ScalableFloat|GetValueatLevel (Variables|Default|GetRow_BouncePads_GroupsToSpawn))) (Variables|Default|GetBGA_SeedPod) _returnvalue_2 (ScalableFloat|GetValueatLevel (Variables|Default|GetRow_BouncePads_RadiusMin)) (ScalableFloat|GetValueatLevel (Variables|Default|GetRow_BouncePads_RadiusMax)) (ScalableFloat|GetValueatLevel (Variables|Default|GetRow_BouncePads_RadialSpacingVarianceMult)) (+ (* (Math|Float|Truncate (ScalableFloat|GetValueatLevel (Variables|Default|GetRow_BouncePads_GroupsToSpawn))) (ScalableFloat|GetValueatLevel (Variables|Default|GetRow_BouncePads_AdditionalHeightPerGroup))) (ScalableFloat|GetValueatLevel (Variables|Default|GetRow_BouncePads_StartingHeight))) (+ (+ (* (Math|Float|Truncate (ScalableFloat|GetValueatLevel (Variables|Default|GetRow_BouncePads_GroupsToSpawn))) (ScalableFloat|GetValueatLevel (Variables|Default|GetRow_BouncePads_AdditionalHeightPerGroup))) (ScalableFloat|GetValueatLevel (Variables|Default|GetRow_BouncePads_StartingHeight))) (ScalableFloat|GetValueatLevel (Variables|Default|GetRow_BouncePads_HeightVariance))) (Variables|Default|GetSizedBGAsNullArray))))

(event Custom|SpawnInitialBGAs
  ;[1 1 1] Spawn Initial BGAs
  (for _array_element (Variables|Default|GetSpawnerParams)
    (for _index (range (.MaxCount_6_670D90C940F89FF560F2F8AA005344F0 (Utilities|Struct|BreakRealityFlowerSpawnerParams _array_element)))
      (CallFunction|SpawnBGA (Utilities|Struct|BreakRealityFlowerSpawnerParams _array_element) (.BGAToSpawn_20_52D78B404E147931A7F9DC8CF2F3D3A3 (Utilities|Struct|BreakRealityFlowerSpawnerParams _array_element)) _index (.MaxCount_6_670D90C940F89FF560F2F8AA005344F0 (Utilities|Struct|BreakRealityFlowerSpawnerParams _array_element)) (.Radius_Min_29_BF3DBAE44CA8EB95100B358C389666F4 (Utilities|Struct|BreakRealityFlowerSpawnerParams _array_element)) (.Radius_Max_30_1C2C45B94151EAD0A3BD89B21EAAD7C7 (Utilities|Struct|BreakRealityFlowerSpawnerParams _array_element)) (.RadialSpacingVarianceMult_35_520DF524427C0770148A0D8CB1C84B0B (Utilities|Struct|BreakRealityFlowerSpawnerParams _array_element)) (.Height_Min_33_2B18E8EA466F9EBADFBDFD950F337C2E (Utilities|Struct|BreakRealityFlowerSpawnerParams _array_element)) (.Height_Max_34_E54989464222D7465AA2488DC44DE9D9 (Utilities|Struct|BreakRealityFlowerSpawnerParams _array_element))))))

(event Custom|SpawnBGA
  (params (Group :type Integer) (BGAToSpawn :type Building%20Gameplay%20Actor%20Class%20Reference) (Index :type Integer) (MaxCount :type Integer) (RadiusMin :type Float%20%28double-precision%29) (RadiusMax :type Float%20%28double-precision%29) (RadialSpacingVarianceMult :type Float%20%28double-precision%29) (HeightMin :type Float%20%28double-precision%29) (HeightMax :type Float%20%28double-precision%29))
  (bind (_returnvalue_x _returnvalue_y) (Math|Vector|BreakVector (* (Math|Random|RandomFloatinRange RadiusMin RadiusMax) (Math|Rotator|GetRotationXVector (Math|Rotator|MakeRotator 0.0 0.0 (+ (* (Math|Conversions|ToFloat(Integer) Index) (* (/ 1.0 (Math|Conversions|ToFloat(Integer) MaxCount)) 3.6e+02)) (* (* (/ 1.0 (Math|Conversions|ToFloat(Integer) MaxCount)) 3.6e+02) (Math|Random|RandomFloatinRange (* -1.0 (Math|Float|Clamp(Float) RadialSpacingVarianceMult)) (Math|Float|Clamp(Float) RadialSpacingVarianceMult)))))))))
  ;[1 1 1] Spawn BGA
  (bind _returnvalue (Building|SpawnBuildingGameplayActorfromWorldContext BGAToSpawn (Math|Transform|MakeTransform (Utilities|Operators|Add (Math|Vector|MakeVector _returnvalue_x _returnvalue_y (Math|Random|RandomFloatinRange HeightMin HeightMax)) (Transformation|GetActorLocation self)) (Math|Rotator|MakeRotator 0.0 0.0 (Math|Random|RandomFloatinRange 0.0 3.6e+02))) self 0 false))
  (bind _result (ArrayMacros|SafeArrayGet (Variables|Default|GetSpawnerParams) Group)
    (:%20%20
      (bind (_result_group_2_2f58375242987eef632d74a2c8966409 _result_bgatospawn_20_52d78b404e147931a7f9dc8cf2f3d3a3 _result_maxcount_6_670d90c940f89ff560f2f8aa005344f0 _result_radius_min_29_bf3dbae44ca8eb95100b358c389666f4 _result_radius_max_30_1c2c45b94151ead0a3bd89b21eaad7c7 _result_radialspacingvariancemult_35_520df524427c0770148a0d8cb1c84b0b _result_height_min_33_2b18e8ea466f9ebadfbdfd950f337c2e _result_height_max_34_e54989464222d7465aa2488dc44de9d9 _result_spawnedbgas_22_02cad072448f2e83e4d1869c42174331) (Utilities|Struct|BreakRealityFlowerSpawnerParams _result))
      (bind _result_2 (ArrayMacros|SafeArrayGet _result_spawnedbgas_22_02cad072448f2e83e4d1869c42174331 Index)
        (:%20%20
          (Utilities|IsValid _result_2
            (:Is%20Valid
              (Gameplay|Die :self _result_2))
            (:Is%20Not%20Valid
              _))
          (Utilities|Array|SetArrayElem _result_spawnedbgas_22_02cad072448f2e83e4d1869c42174331 Index _returnvalue true))))))

(fn CreateNewSizedNullBGAArray
  (params (Size :type Integer))
  ;[1 1 1] Create a new correctly sized array for each element of SpawnerParams->SpawnedBGAs to avoid errors of having incorrectly sized arrays when adding newly spawned actors
  (Variables|Default|SetSizedBGAsNullArray (Utilities|Array|MakeArray 0))
  (for _ (range 2 (+ Size 1))
    (Utilities|Array|Add (Variables|Default|GetSizedBGAsNullArray)))
  (return))

(fn ConstructionScript
  (|CallParentFunction))
