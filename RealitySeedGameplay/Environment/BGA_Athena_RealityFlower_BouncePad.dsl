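; Bounce pad gameplay actor for the Reality Flower environment prop that applies gameplay effects to player pawns touching it.
; On Sphere begin overlap, it casts the other actor to a Fort player pawn and removes active effects granting a configured tag.
; It then applies a configurable effect container to that pawn through the pawn's own ability system component.
; No launch logic appears in this graph.
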(event EventBeginPlay)

(event Collision|EventActorBeginOverlap
  (params (OtherActor :type Actor%20Object%20Reference)))

(event EventTick
  (params (DeltaSeconds :type Float%20%28single-precision%29)))

(event OnComponentBeginOverlap(Sphere)
  (params (OverlappedComponent :type Primitive%20Component%20Object%20Reference) (OtherActor :type Actor%20Object%20Reference) (OtherComp :type Primitive%20Component%20Object%20Reference) (OtherBodyIndex :type Integer) (bFromSweep :type Boolean) (SweepResult :type Hit%20Result%20Structure%20%28by%20ref%29))
  (bind _asfort_player_pawn (Utilities|Casting|CastToFortPlayerPawn OtherActor))
  (GameplayEffects|RemoveActiveEffectswithGrantedTags (Ability|GetFortAbilitySystemComponent _asfort_player_pawn) (Variables|Default|GetTag_GE_GrantedTagToRemove))
  (Fortnite|ApplyGameplayEffectsToActor _asfort_player_pawn (Variables|Default|GetGEC_EffectsToApply) (Transformation|GetActorLocation _asfort_player_pawn) false 0 (Ability|GetFortAbilitySystemComponent _asfort_player_pawn)))

(fn ConstructionScript)
