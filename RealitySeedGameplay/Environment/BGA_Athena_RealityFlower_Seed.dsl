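; Reality Flower seed actor that floats through the world as a projectile and can be collected by interaction.
; InitialVelocity launches it server-side at random pitch, with a timeline blending max speed and gravity to final values.
; Timed server drift picks random directions, steering back toward spawn outside a leash radius.
; Interaction grants the seed item, while lifespan expiry or floor-like bounces spawn a pickup; other bounces play throttled water/world cues.
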(event Custom|SpawnItemPickupAndDie
  ;[1 1 1] Spawn Item Pickup And Die
  (if (not (Variables|Default|GetHasSpawnedItem))
    (Variables|Default|SetHasSpawnedItem true)
    (Inventory|SpawnPickupInWorld (Variables|Default|GetItem_Seed) 1 (Transformation|GetActorLocation self) "0.000000,0.000000,1.000000" 0 true true true -1 "Destruction" "ItemSpawner" 0 false)
    (GameplayCue|ExecuteGameplayCue self (Variables|Effects|GetGC_Destroyed))
    (Gameplay|Die)
    (Components|DestroyComponent (Variables|Default|GetNS_RealitySeed_Trail))))

(event Custom|PlayFX_Idle
  ;[1 1 1] Play FX: Idle
  (GameplayCue|AddGameplayCue self (Variables|Default|GetGC_InAir))
  (CallFunction|SpinSeed))

(event Custom|DoRandomDriftAfterRandomTime
  ;[1 1 1] Do Random Drift After Random Time
  (Utilities|Time|SetTimerbyEvent (EventDispatchers|CreateEvent :!event Custom|RandomDrift) (Math|Random|RandomFloatinRange (ScalableFloat|GetValueatLevel (Variables|Balance|GetRow_Drift_TimeBetweenDirectionChanges_Min)) (ScalableFloat|GetValueatLevel (Variables|Balance|GetRow_Drift_TimeBetweenDirectionChanges_Max)))))

(event Custom|SpinSeed
  (bind _staticmesh (Variables|Default|GetStaticMesh))
  ;[1 1 1] Timeline for spinning the seed.
  (Components|Timeline|SetPlayRate (Variables|BGA_Athena_RealityFlower_Seed|GetSpeedSpinTimeline) (Variables|Default|GetSeedSpinPlayRate)) ; Timeline is 1 second duration
  (bind (_direction _newtrack_0) (Timeline|SpeedSpinTimeline PlayFromStart)
    (:Update
      (Rendering|Material|SetScalarParameterValueonMaterials _staticmesh "RotationAngle" (* _newtrack_0 (Variables|Default|GetSeedSpinSpeed)))
      (Rendering|Material|SetScalarParameterValueonMaterials _staticmesh "MotionRandomInput" (Math|Random|RandomFloat)))))

(event EventBeginPlay
  ;[1 1 1] Begin Play
  (|CallParentFunction)
  (Game|Components|ProjectileMovement|Interpolation|SetInterpolatedComponent (Variables|Default|GetFortProjectileMovement) (Variables|Default|GetStaticMesh))
  (CallFunction|PlayFXIdle)
  (Variables|Default|SetLocation_Start (Transformation|GetActorLocation self))
  (Utilities|Time|SetTimerbyEvent (EventDispatchers|CreateEvent :!event Custom|RandomDrift) (ScalableFloat|GetValueatLevel (Variables|Balance|GetRow_Initial_SpeedDecayDuration)))
  (Utilities|Time|SetTimerbyEvent (EventDispatchers|CreateEvent :!event Custom|SpawnItemPickupAndDie) (ScalableFloat|GetValueatLevel (Variables|Balance|GetRow_MaxLifespan)))
  ;[1 0.7581 0] Debug
  (if (Variables|Default|GetDebug)
    (if (Networking|IsServer)
      (Utilities|Time|SetTimerbyEvent (delegate Custom|DebugServerLocation) 0.1 true)
      (else
        _))
    (Rendering|Debug|DrawDebugSphere (Variables|Default|GetLocation_Start) (ScalableFloat|GetValueatLevel (Variables|Balance|GetRow_Leash_Radius)) 36 (Math|Color|MakeColor 0.0 0.0 64.0) 999.0 0.5)
    (Rendering|Debug|DrawDebugSphere (Variables|Default|GetLocation_Start) 16.0 12 "(R=1.000000,G=1.000000,B=1.000000,A=1.000000)" 999.0 1.0)))

(event Custom|RandomDrift
  (bind (_returnvalue_x _returnvalue_y) (Math|Vector|BreakVector (Math|Random|RandomUnitVector)))
  (bind _self self)
  (bind _returnvalue (Transformation|GetActorLocation _self))
  (bind _location_start (Variables|Default|GetLocation_Start))
  (bind _returnvalue_2 (- _location_start _returnvalue))
  ;[1 1 1] Random Drift
  (if (Networking|IsServer)
    (if (or (< (Math|Vector|VectorLength (- (Transformation|GetActorLocation self) (Variables|Default|GetLocation_Start))) (ScalableFloat|GetValueatLevel (Variables|Balance|GetRow_Leash_Radius))) (< (ScalableFloat|GetValueatLevel (Variables|Balance|GetRow_Leash_Radius)) 0))
      (bind _output_get (Variables|Default|SetRandomUnitVector (Math|Vector|Normalize (Utilities|Operators|Add (Math|Vector|Normalize (Math|Vector|MakeVector _returnvalue_x _returnvalue_y)) (Math|Vector|MakeVector 0.0 0.0 (Math|Random|RandomFloatinRange (ScalableFloat|GetValueatLevel (Variables|Balance|GetRow_Drift_ZComponent_Min)) (ScalableFloat|GetValueatLevel (Variables|Balance|GetRow_Drift_ZComponent_Max))))))))
      (Projectile|SetDesiredDirection (Variables|Default|GetFortProjectileMovement) _output_get :!class /Script/FortniteGame.FortProjectileMovementComponent)
      (CallFunction|DoRandomDriftAfterRandomTime)
      ;[1 0.7581 0] Debug
      (if (Variables|Default|GetDebug)
        (Rendering|Debug|DrawDebugLine (Transformation|GetActorLocation self) (+ (Transformation|GetActorLocation self) (* (Variables|Default|GetRandomUnitVector) "256.000000,256.000000,256.000000")) (Math|Color|MakeColor 0.0 255.0) 3.0 2.0))
      (else
        (Projectile|SetDesiredDirection (Variables|Default|GetFortProjectileMovement) _returnvalue_2 :!class /Script/FortniteGame.FortProjectileMovementComponent)
        (CallFunction|DoRandomDriftAfterRandomTime)
        ;[1 0.7581 0] Debug
        (if (Variables|Default|GetDebug)
          (Rendering|Debug|DrawDebugLine (Transformation|GetActorLocation self) (+ (* (Math|Vector|Normalize _returnvalue_2) "256.000000,256.000000,256.000000") (Transformation|GetActorLocation self)) (Math|Color|MakeColor 255.0 255.0) 3.0 2.0))))))

(event Custom|DebugServerLocation
  ;[1 1 1] Begin Play
  ;[1 0.7581 0] Debug
  (Debug|DebugSphereSendtoAll (Transformation|GetActorLocation) 1e+02 12 (Utilities|Struct|MakeColor 0 255 255) false 0.2 0 2.0))

(event Interaction|EventBlueprintOnInteract
  (params (InteractingPawn :type Fort%20Pawn%20Object%20Reference) (InteractionBeingAttempted :type EInteractionBeingAttempted%20Enum :default "FirstInteraction"))
  (bind _returnvalue (Fortnite|GetFortPlayerControllerfromActor InteractingPawn))
  ;[1 1 1] On Interact (Server)
  (|CallParentFunction InteractingPawn InteractionBeingAttempted)
  (if (not (Variables|Default|GetHasSpawnedItem))
    (Utilities|IsValid _returnvalue
      (:Is%20Valid
        (Variables|Default|SetHasSpawnedItem true)
        (Inventory|GiveItemtoInventoryOwner :InventoryOwner _returnvalue :ItemDefinition (Variables|Default|GetItem_Seed) :NumberToGive 1 :bNotifyPlayer true :ItemLevel -1 :PickupInstigatorHandle -1 :bUseItemPickupAnalyticEvent true :WeaponAmmoOverride -1)
        (GameplayCue|ExecuteGameplayCue self (Variables|Effects|GetGC_Collect))
        (Ability|SendGameplayEventtoActor _returnvalue (Variables|Default|GetEvent%20Tag) (Utilities|Struct|MakeGameplayEventData (Variables|Default|GetEvent%20Tag) _returnvalue))
        (Gameplay|Die :Killer _returnvalue)))))

(event EventOnDeathServer
  (params (Damage :type Float%20%28single-precision%29) (DamageTags :type Gameplay%20Tag%20Container%20Structure%20%28by%20ref%29) (Momentum :type Vector :default "0, 0, 0") (HitInfo :type Hit%20Result%20Structure%20%28by%20ref%29) (InstigatedBy :type Controller%20Object%20Reference) (DamageCauser :type Actor%20Object%20Reference) (EffectContext :type Gameplay%20Effect%20Context%20Handle%20Structure))
  ;[1 1 1] On Death (Server)
  (|CallParentFunction Damage DamageTags Momentum HitInfo InstigatedBy DamageCauser EffectContext)
  (Rendering|SetActorHiddenInGame true)
  (Components|DestroyComponent (Variables|Default|GetNS_RealitySeed_Trail)))

(event Custom|InitialVelocity
  (bind (_returnvalue_roll _returnvalue_pitch _returnvalue_yaw) (Math|Rotator|BreakRotator (Transformation|GetActorRotation)))
  ;[1 1 1] Initial Velocity
  (if (Networking|IsServer)
    (CallFunction|RepSetInitialVelocity)
    (bind _output_get (Variables|Default|SetInitialVelocityRep (* (ScalableFloat|GetValueatLevel (Variables|Balance|GetRow_Initial_Speed)) (Math|Rotator|GetRotationXVector (Math|Rotator|MakeRotator _returnvalue_roll (Math|Random|RandomFloatinRange (ScalableFloat|GetValueatLevel (Variables|Balance|GetRow_Initial_Velocity_Pitch_Min)) (ScalableFloat|GetValueatLevel (Variables|Balance|GetRow_Initial_Velocity_Pitch_Max))) _returnvalue_yaw)))))
    (Projectile|SetVelocity (Variables|Default|GetFortProjectileMovement) _output_get)
    (Components|Timeline|SetPlayRate (Variables|BGA_Athena_RealityFlower_Seed|GetInitialSpeedDecay) (/ 1.0 (ScalableFloat|GetValueatLevel (Variables|Balance|GetRow_Initial_SpeedDecayDuration))))
    (bind (_direction _alpha) (Timeline|InitialSpeedDecay PlayFromStart)
      (:Update
        (CallFunction|SetMaxSpeedandGavitybyAlpha _alpha))
      (:Finished
        (CallFunction|SetMaxSpeedandGavitybyAlpha 1.0)
        (Variables|Default|SetRotationRate 0.5)
        (Networking|ForceNetUpdate)))))

(event Custom|RepSetInitialVelocity
  ;[1 1 1] Initial Velocity
  ;;[1 1 1] Rep Set Initial Velocity
  (Variables|/Script/Engine.ProjectileMovementComponent|SetMaxSpeed (ScalableFloat|GetValueatLevel (Variables|Balance|GetRow_Initial_Speed)) (Variables|Default|GetFortProjectileMovement))
  (Variables|/Script/Engine.ProjectileMovementComponent|SetProjectileGravityScale (ScalableFloat|GetValueatLevel (Variables|Balance|GetRow_Initial_GravityScale)) (Variables|Default|GetFortProjectileMovement)))

(event OnProjectileBounce(FortProjectileMovement)
  (params (ImpactResult :type Hit%20Result%20Structure%20%28by%20ref%29) (ImpactVelocity_X :type Float%20%28double-precision%29) (ImpactVelocity_Y :type Float%20%28double-precision%29) (ImpactVelocity_Z :type Float%20%28double-precision%29))
  (bind (_bblockinghit _binitialoverlap _time _distance _location _impactpoint _normal _impactnormal _physmat _hitactor) (Collision|BreakHitResult ImpactResult))
  (bind (_normal_x _normal_y _normal_z) (Math|Vector|BreakVector _normal))
  (bind (_aswater_body _bsuccess) (Utilities|Casting|CastToWaterBody _hitactor))
  ;[1 1 1] On Bounce
  (if (and (>= _normal_z (Math|Float|Min(Float) (ScalableFloat|GetValueatLevel (Variables|Balance|GetRow_Bounce_ZNormalMinThresholdToPop)) 1.0)) (Math|Float|InRange(Float) ImpactVelocity_Z (Math|Float|Min(Float) (ScalableFloat|GetValueatLevel (Variables|Balance|GetRow_Bounce_ZImpactVelocityToPop_Min))) 0.0))
    (CallFunction|SpawnItemPickupandDie)
    (elif (Variables|Default|GetBounceFXOnCooldown)
      _
      (else
        (CallFunction|SetBounceFXCooldown)
        (if _bsuccess
          (Ability|ExecuteGameplayCueWithParamsReplicated (Variables|Default|GetGC_HitWater) (GameplayCue|MakeGameplayCueParametersfromHitResult ImpactResult))
          (else
            (Ability|ExecuteGameplayCueWithParamsReplicated (Variables|Default|GetGC_HitWorld) (GameplayCue|MakeGameplayCueParametersfromHitResult ImpactResult))))))))

(event Custom|SetBounceFXCooldown
  ;[1 1 1] Set Bounce FX Cooldown
  (Variables|Default|SetBounceFXOnCooldown true)
  (Utilities|FlowControl|Delay (ScalableFloat|GetValueatLevel (Variables|Default|GetRow_BounceFXCooldown)))
  (Variables|Default|SetBounceFXOnCooldown false))

(fn BlueprintCanInteract
  (params (InteractingPawn :type Fort%20Pawn%20Object%20Reference) (InteractionBeingAttempted :type EInteractionBeingAttempted%20Enum :default "FirstInteraction") (InteractionType :type TInteractionType%20Enum :default "IT_NoInteraction"))
  (return true))

(fn BlueprintGetInteractionString
  (params (InteractingPawn :type Fort%20Pawn%20Object%20Reference) (InteractionBeingAttempted :type EInteractionBeingAttempted%20Enum :default "FirstInteraction"))
  (|CallParentFunction InteractingPawn InteractionBeingAttempted)
  (return (Variables|Default|GetText_Interact)))

(fn BlueprintGetInteractionTime
  (params (InteractingPawn :type Fort%20Pawn%20Object%20Reference) (InteractionBeingAttempted :type EInteractionBeingAttempted%20Enum :default "FirstInteraction"))
  (|CallParentFunction InteractingPawn InteractionBeingAttempted)
  (return (ScalableFloat|GetValueatLevel (Variables|Balance|GetRow_InteractTime)) true))

(fn SetMaxSpeedAndGavityByAlpha
  (params (Alpha :type Float%20%28double-precision%29))
  (Variables|/Script/Engine.ProjectileMovementComponent|SetMaxSpeed (Math|Float|Lerp (ScalableFloat|GetValueatLevel (Variables|Balance|GetRow_Initial_Speed)) (ScalableFloat|GetValueatLevel (Variables|Balance|GetRow_Final_Speed)) Alpha) (Variables|Default|GetFortProjectileMovement))
  (Variables|/Script/Engine.ProjectileMovementComponent|SetProjectileGravityScale (Math|Float|Lerp (ScalableFloat|GetValueatLevel (Variables|Balance|GetRow_Initial_GravityScale)) (ScalableFloat|GetValueatLevel (Variables|Balance|GetRow_Final_GravityScale)) Alpha) (Variables|Default|GetFortProjectileMovement)))

(fn ConstructionScript
  (|CallParentFunction)
  (Components|Sphere|SetSphereRadius (Variables|Default|GetSphere) (ScalableFloat|GetValueatLevel (Variables|Balance|GetRow_SeedCollisionRadius)) true)
  (Variables|/Script/Engine.ProjectileMovementComponent|SetMaxSpeed (ScalableFloat|GetValueatLevel (Variables|Balance|GetRow_Final_Speed)) (Variables|Default|GetFortProjectileMovement))
  (Projectile|SetAccelerationMagnitude (Variables|Default|GetFortProjectileMovement) (ScalableFloat|GetValueatLevel (Variables|Balance|GetRow_Acceleration)) :!class /Script/FortniteGame.FortProjectileMovementComponent)
  (Variables|/Script/Engine.ProjectileMovementComponent|SetBounciness (ScalableFloat|GetValueatLevel (Variables|Balance|GetRow_Bounciness)) (Variables|Default|GetFortProjectileMovement))
  (Variables|/Script/Engine.ProjectileMovementComponent|SetFriction (ScalableFloat|GetValueatLevel (Variables|Balance|GetRow_BounceFriction)) (Variables|Default|GetFortProjectileMovement)))
