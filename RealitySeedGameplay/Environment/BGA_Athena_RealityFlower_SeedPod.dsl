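; Reality flower seed pod building actor that grows on spawn through a timeline-driven, rep-notified mesh scale.
; On BeginPlay the server rolls a seed count from scalable-float balance rows.
; Server damage fires throttled impact cues and ignites the pod through Curie for matching damage tags.
; Death, also triggered by linked-actor destruction, hides it and, unless effects are skipped, launches radially jittered seeds and optionally triggers an explosion ability.
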(event Building|EventOnBuildingActorInitialized
  (params (InitializationReason :type EFortBuildingInitializationReason%20Enum :default "StaticallyPlaced") (BuildingPersistentState :type EFortBuildingPersistentState%20Enum :default "Default"))
  ;[1 1 1] On Building Actor Initialized
  (|CallParentFunction InitializationReason BuildingPersistentState)
  (CallFunction|GrowonSpawn))

(event Curie|EventOnCurieElementInteract
  (params (CurieContainerHandle :type Curie%20Container%20Handle%20Structure) (ElementTag :type Gameplay%20Tag%20Structure) (InteractParamsHandle :type Curie%20Interact%20Params%20Handle%20Structure%20%28by%20ref%29))
  ;[1 1 1] On Curie Interact
  (|CallParentFunction CurieContainerHandle ElementTag InteractParamsHandle)
  (if (GameplayTags|Equal(GameplayTag) ElementTag "(TagName=\"Curie.Element.Fire\")")
    (Development|PrintString "Fire Interact" true true "(R=1.000000,G=0.000000,B=0.036456,A=1.000000)" 5.0)))

(event EventBeginPlay
  ;[1 1 1] On Begin Play
  (|CallParentFunction)
  (Ability|EquipFortAbilitySet self (Variables|Default|GetGAS_AbilitySet) self)
  (if (Networking|IsServer)
    (Utilities|FlowControl|DoOnce false
      (:Completed
        (Variables|Default|SetSeedsToSpawnOnDeath (Math|Random|RandomIntegerinRange (Math|Integer|Min(Integer) (Math|Float|Truncate (ScalableFloat|GetValueatLevel (Variables|Balance|GetRow_SeedsToSpawnOnDeath_Min))) (Math|Float|Truncate (ScalableFloat|GetValueatLevel (Variables|Balance|GetRow_SeedsToSpawnOnDeath_Max)))) (Math|Integer|Max(Integer) (Math|Float|Truncate (ScalableFloat|GetValueatLevel (Variables|Balance|GetRow_SeedsToSpawnOnDeath_Min))) (Math|Float|Truncate (ScalableFloat|GetValueatLevel (Variables|Balance|GetRow_SeedsToSpawnOnDeath_Max))))))))))

(event Curie|EventOnCurieElementInteractBegun
  (params (CurieContainerHandle :type Curie%20Container%20Handle%20Structure) (ElementTag :type Gameplay%20Tag%20Structure) (InteractParamsHandle :type Curie%20Interact%20Params%20Handle%20Structure%20%28by%20ref%29))
  ;[1 1 1] On Curie Interact
  (|CallParentFunction CurieContainerHandle ElementTag InteractParamsHandle)
  (if (GameplayTags|Equal(GameplayTag) ElementTag "(TagName=\"Curie.Element.Fire\")")
    (Development|PrintString "Fire Begun" true true "(R=1.000000,G=0.000000,B=0.036456,A=1.000000)" 5.0)))

(event Custom|GrowOnSpawn
  ;[1 1 1] Grow On Spawn
  (bind (_direction _sizepercent) (Timeline|Grow PlayFromStart)
    (:Update
      (Variables|Default|SetMesh_Scale (* _sizepercent (Variables|Default|GetMesh_MaxScale))))))

(event EventOnDeathServer
  (params (Damage :type Float%20%28single-precision%29) (DamageTags :type Gameplay%20Tag%20Container%20Structure%20%28by%20ref%29) (Momentum :type Vector :default "0, 0, 0") (HitInfo :type Hit%20Result%20Structure%20%28by%20ref%29) (InstigatedBy :type Controller%20Object%20Reference) (DamageCauser :type Actor%20Object%20Reference) (EffectContext :type Gameplay%20Effect%20Context%20Handle%20Structure))
  ;[1 1 1] On Death (Server)
  (|CallParentFunction Damage DamageTags Momentum HitInfo InstigatedBy DamageCauser EffectContext)
  (Rendering|SetActorHiddenInGame true)
  (Variables|Default|SetCollisionEnabled false)
  (if (not (Variables|Default|GetSkipDeathEffects))
    (CallFunction|SpawnSeedsonDeath)
    (if (ScalableFloat|GetValueasBool (Variables|Balance|GetRow_bExplodeOnDeath))
      (Ability|SendGameplayEventtoActor self "(TagName=\"Abilities.Generic.TriggeredAbility.Activate\")" (Utilities|Struct|MakeGameplayEventData "(TagName=\"\")" self)) ; Explode
    )))

(event Custom|SpawnSeedsOnDeath
  ;[1 1 1] Spawn Seeds On Death
  (if (> (Variables|Default|GetSeedsToSpawnOnDeath) 0)
    (for _index (range (Variables|Default|GetSeedsToSpawnOnDeath))
      (bind _returnvalue (Building|SpawnBuildingGameplayActorfromWorldContext (Variables|Default|GetBGA_RealitySeed) (CallFunction|GetSpawnedSeedTransform _index) self))
      (bind _asbga_athena_reality_flower_seed (Utilities|Casting|CastToBGA_Athena_RealityFlower_Seed _returnvalue))
      (Class|BGAAthenaRealityFlowerSeed|InitialVelocity _asbga_athena_reality_flower_seed))))

(event EventOnDeathPlayEffects
  (params (Damage :type Float%20%28single-precision%29) (DamageTags :type Gameplay%20Tag%20Container%20Structure%20%28by%20ref%29) (Momentum :type Vector :default "0, 0, 0") (HitInfo :type Hit%20Result%20Structure%20%28by%20ref%29) (InstigatedBy :type Fort%20Pawn%20Object%20Reference) (DamageCauser :type Actor%20Object%20Reference) (EffectContext :type Gameplay%20Effect%20Context%20Handle%20Structure))
  (bind (_returnvalue_location _returnvalue_rotation) (Math|Transform|BreakTransform (Transformation|GetSocketTransform (Variables|Debug|GetStaticMeshComponent) "FX Spawn" "RTS_World")))
  ;[1 1 1] On Death (Client)
  (|CallParentFunction Damage DamageTags Momentum HitInfo InstigatedBy DamageCauser EffectContext)
  (Audio|Components|Audio|Stop (Variables|Default|GetAudio))
  (GameplayCue|ExecuteGameplayCue%28Burst%29Local self (Variables|Effects|GetGC_Die) (Ability|GameplayCue|MakeGameplayCueParameters :NormalizedMagnitude 0.0 :RawMagnitude 0.0 :Location _returnvalue_location :Normal (Math|Rotator|GetRotationXVector _returnvalue_rotation) :GameplayEffectLevel 1 :AbilityLevel 1 :bReplicateLocationWhenUsingMinimalRepProxy false)))

(event EventOnDamageServer
  (params (Damage :type Float%20%28single-precision%29) (DamageTags :type Gameplay%20Tag%20Container%20Structure%20%28by%20ref%29) (Momentum :type Vector :default "0, 0, 0") (HitInfo :type Hit%20Result%20Structure%20%28by%20ref%29) (InstigatedBy :type Controller%20Object%20Reference) (DamageCauser :type Actor%20Object%20Reference) (EffectContext :type Gameplay%20Effect%20Context%20Handle%20Structure))
  (bind (_bblockinghit _binitialoverlap _time _distance _location _impactpoint _normal _impactnormal _physmat _hitactor _hitcomponent) (Collision|BreakHitResult HitInfo))
  (bind _returnvalue (Utilities|Time|GetGameTimeinSeconds))
  ;[1 1 1] On Damage (Server)
  (|CallParentFunction Damage DamageTags Momentum HitInfo InstigatedBy DamageCauser EffectContext)
  (if (>= (- _returnvalue (Variables|Default|GetImpact_TimeStamp)) 0.133336)
    (Variables|Default|SetImpact_TimeStamp _returnvalue)
    (GameplayCue|ExecuteGameplayCueOnActor(Burst) self (Variables|Effects|GetGC_Damaged) (Ability|GameplayCue|MakeGameplayCueParameters :NormalizedMagnitude 0.0 :RawMagnitude Damage :EffectContext (Class|AthenaFunctionLibrary|MakeSimpleFortEffectContext InstigatedBy DamageCauser) :AggregatedSourceTags DamageTags :Location _location :Normal _normal :Instigator InstigatedBy :EffectCauser DamageCauser :SourceObject _hitactor :PhysicalMaterial _physmat :GameplayEffectLevel 1 :AbilityLevel 1 :TargetAttachComponent _hitcomponent :bReplicateLocationWhenUsingMinimalRepProxy false))
    ;[1 0 0] HACKY - Light self on fire if hit by explosive damage
    (if (GameplayTags|HasAnyTags DamageTags (Variables|Default|GetTags_IncomingDamageToIgnite) true)
      (Curie|ExecuteFireApplication InstigatedBy DamageCauser self 1e+02 HitInfo))))

(event Custom|Trigger%20FX%20Pulse
  ;[1 1 1] Impact FX on Pod
  (Utilities|FlowControl|DoOnce false
    (:Completed
      (Niagara|SetNiagaraVariable(Bool) (Variables|Default|GetNS_RealityPod_Idle) "Hit" true)
      (Components|Activation|Activate (Variables|Default|GetNS_RealityPod_Pulse) true)
      (Utilities|FlowControl|Delay 0.25)
      (Niagara|SetNiagaraVariable(Bool) (Variables|Default|GetNS_RealityPod_Idle) "Hit" false))))

(event OnLinkedActorDestroyed(FortLinkToActor)
  (params (DamageCauser :type Actor%20Object%20Reference))
  ;[1 1 1] On Linked Actor Destroyed, Die
  (Gameplay|Die))

(fn GetMarkerDetailsId
  (return "MarkerDetails.RealitySeedPod"))

(fn GetSpawnedSeedTransform
  (params (SeedIndex :type Integer))
  (bind (_origin _boxextent) (Collision|GetActorBounds false))
  (bind (_boxextent_x _boxextent_y _boxextent_z) (Math|Vector|BreakVector _boxextent))
  (return (Math|Transform|MakeTransform (+ (* (+ (Math|Float|Max(Float) (ScalableFloat|GetValueatLevel (Variables|Balance|GetRow_SeedCollisionRadius)) _boxextent_z) 4.0) (Math|Vector|GetUpVector (Transformation|GetActorRotation))) (Transformation|GetActorLocation)) (Math|Rotator|MakeRotator 0.0 0.0 (CallFunction|GetSpawnedSeedYaw SeedIndex)) "1.000000,1.000000,1.000000")))

(fn GetSpawnedSeedYaw
  (params (SpawnedSeedIndex :type Integer))
  (return (+ (* (Math|Conversions|ToFloat(Integer) SpawnedSeedIndex) (* (/ 1.0 (Math|Conversions|ToFloat(Integer) (Variables|Default|GetSeedsToSpawnOnDeath))) 3.6e+02)) (* (* (/ 1.0 (Math|Conversions|ToFloat(Integer) (Variables|Default|GetSeedsToSpawnOnDeath))) 3.6e+02) (Math|Random|RandomFloatinRange (* -1.0 (Math|Float|Clamp(Float) (ScalableFloat|GetValueatLevel (Variables|Balance|GetRow_Seeds_RadialSpacingVarianceMult)))) (Math|Float|Clamp(Float) (ScalableFloat|GetValueatLevel (Variables|Balance|GetRow_Seeds_RadialSpacingVarianceMult))))))))

(fn GetTargetTagsForQuestEvent
  (return 0))

(fn OnRep_CollisionEnabled
  (Collision|SetActorEnableCollision (Variables|Default|GetCollisionEnabled)))

(fn OnRep_Mesh_Scale
  (Transformation|SetWorldScale3D (Variables|Debug|GetStaticMeshComponent) (Math|Conversions|ToVector(Float) (Variables|Default|GetMesh_Scale))))

(fn ConstructionScript
  (|CallParentFunction))
