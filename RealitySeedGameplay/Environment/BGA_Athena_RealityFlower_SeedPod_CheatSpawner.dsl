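; BGA_Athena_RealityFlower_SeedPod_CheatSpawner is a cheat helper actor that spawns a Reality Flower seed pod in the RealitySeed gameplay feature.
; On BeginPlay, with authority only, it prints a debug string and spawns BGA_Athena_RealityFlower_SeedPod at its own transform.
; It then destroys itself, acting as a one-shot placeholder for the real pod.
; The construction script is empty.
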(event EventBeginPlay
  (Networking|SwitchHasAuthority
    (:Authority
      (Development|PrintString "Spawn Seed Pod")
      (bind _returnvalue (Game|SpawnActorfromClass "/RealitySeedGameplay/Environment/BGA_Athena_RealityFlower_SeedPod.BGA_Athena_RealityFlower_SeedPod_C" (Transformation|GetActorTransform) "Undefined" "MultiplyWithRoot" 0 0 0 0))
      (Actor|DestroyActor))))

(fn ConstructionScript)
