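; Level-placed manager that randomizes how many linked Reality Flower seed pods remain in a match by culling pre-placed pods.
; When the playlist is ready on the server, it picks a count from balance-row min/max, shuffles the pods, silently destroys the surplus, then itself.
; A debug flag skips culling to keep every pod.
; Its construction script caches pod locations and draws editor lines, spheres and count text for level designers.
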(event EventBeginPlay
  ;[1 1 1] On Begin Play and Playlist ready (server), calculate seed pods to spawn and destroy any excess
  (|CallParentFunction)
  (if (and (not (Variables|Default|GetDebugTest_SpawnAllPods)) (Networking|IsServer))
    (Class|FortAsyncActionCurrentPlaylistReady|PlaylistDataReadyAsync
      (:OnReady
        (if (< (ScalableFloat|GetValueatLevel (Variables|Balance|GetRow_SeedPodsToSpawn_Min)) 0)
          (Variables|Default|SetSeedPodsToSpawn (Utilities|Array|Length (Variables|Default|GetSeedPods)))
          (else
            (Variables|Default|SetSeedPodsToSpawn (Math|Random|RandomIntegerinRange (Math|Integer|Min(Integer) (Math|Float|Truncate (ScalableFloat|GetValueatLevel (Variables|Balance|GetRow_SeedPodsToSpawn_Min))) (Math|Integer|SelectInt (Math|Integer|Min(Integer) (Math|Float|Truncate (ScalableFloat|GetValueatLevel (Variables|Balance|GetRow_SeedPodsToSpawn_Max))) (Utilities|Array|Length (Variables|Default|GetSeedPods))) (Utilities|Array|Length (Variables|Default|GetSeedPods)) (>= (ScalableFloat|GetValueatLevel (Variables|Balance|GetRow_SeedPodsToSpawn_Max)) 0))) (Math|Integer|Max(Integer) (Math|Float|Truncate (ScalableFloat|GetValueatLevel (Variables|Balance|GetRow_SeedPodsToSpawn_Min))) (Math|Integer|SelectInt (Math|Integer|Min(Integer) (Math|Float|Truncate (ScalableFloat|GetValueatLevel (Variables|Balance|GetRow_SeedPodsToSpawn_Max))) (Utilities|Array|Length (Variables|Default|GetSeedPods))) (Utilities|Array|Length (Variables|Default|GetSeedPods)) (>= (ScalableFloat|GetValueatLevel (Variables|Balance|GetRow_SeedPodsToSpawn_Max)) 0)))))))
        (if (< (Variables|Default|GetSeedPodsToSpawn) (Utilities|Array|Length (Variables|Default|GetSeedPods)))
          (Utilities|Array|Shuffle (Variables|Default|GetSeedPods))
          (Variables|Default|SetSeedPodsToDestroy (- (Utilities|Array|Length (Variables|Default|GetSeedPods)) (Variables|Default|GetSeedPodsToSpawn)))
          (bind (_array_element _array_index) (Utilities|Array|ForEachLoopwithBreak (Variables|Default|GetSeedPods))
            (:LoopBody
              (if (< _array_index (Variables|Default|GetSeedPodsToDestroy))
                _
                (else
                  _))
              (Variables|/RealitySeedGameplay/Environment/BGA_Athena_RealityFlower_SeedPod.BGA_Athena_RealityFlower_SeedPod_C|SetSkipDeathEffects true _array_element)
              (Actor|DestroyActor _array_element))
            (:Completed
              _))
          (else
            _))
        (Actor|DestroyActor)))))

(event Custom|DebugSeedPodManager
  (bind _seedpods (Variables|Default|GetSeedPods))
  (bind _returnvalue (Utilities|Array|Length _seedpods))
  (bind _returnvalue_2 (Utilities|String|ToString(Integer) _returnvalue))
  (bind _row_seedpodstospawn_max (Variables|Balance|GetRow_SeedPodsToSpawn_Max))
  (bind _returnvalue_3 (ScalableFloat|GetValueatLevel _row_seedpodstospawn_max))
  (bind _returnvalue_4 (Math|Color|MakeColor 0.0 255.0))
  (bind _returnvalue_5 (Math|Float|Truncate _returnvalue_3))
  (bind _returnvalue_6 (Utilities|String|ToString(Integer) _returnvalue_5))
  (bind _row_seedpodstospawn_min (Variables|Balance|GetRow_SeedPodsToSpawn_Min))
  (bind _returnvalue_7 (ScalableFloat|GetValueatLevel _row_seedpodstospawn_min))
  (bind _returnvalue_8 (Math|Float|Truncate _returnvalue_7))
  (bind _returnvalue_9 (Utilities|String|ToString(Integer) _returnvalue_8))
  (bind _debugtest_spawnallpods (Variables|Default|GetDebugTest_SpawnAllPods))
  (bind _returnvalue_10 (Utilities|String|SelectString "<br>WARNING: Debug Test Spawn All Pods = TRUE. Set to FALSE for final build" "" _debugtest_spawnallpods))
  (bind _returnvalue_11 (Utilities|String|Append "Seed Pod Spawn Manager<br>Linked Count = " _returnvalue_2 "<br>Spawn Count = " _returnvalue_9 " - " _returnvalue_6 "<br>Move To See Linked Pods. Save Level to clear debug lines." _returnvalue_10))
  ;[1 1 1] Debug Seed Pod Manager - Draw lines to linked pods and update text. Called on Construction for LD visualization
  (if (Editor|InEditororPIE)
    (for _array_element (Variables|Default|GetLinkedSeedPodLocations)
      (Rendering|Debug|DrawDebugLine _array_element (Transformation|GetActorLocation self) _returnvalue_4 1e+01 1e+01)
      (Rendering|Debug|DrawDebugSphere _array_element 165.0 12 _returnvalue_4 1e+01 1e+01))
    (Rendering|Components|TextRender|SetText (Variables|Default|GetTextRender) (Utilities|Text|ToText(String) _returnvalue_11))
    (Development|PrintString _returnvalue_11 true true "(R=0.000000,G=1.000000,B=0.002132,A=1.000000)" 1e+01)))

(fn CreateLinkedSeedPodLocationArray
  (Utilities|Array|Clear (Variables|Default|GetLinkedSeedPodLocations))
  (for _array_element (Variables|Default|GetSeedPods)
    (Utilities|IsValid _array_element
      (:Is%20Valid
        (Utilities|Array|Add (Variables|Default|GetLinkedSeedPodLocations) (Transformation|GetActorLocation _array_element))))))

(fn ShouldDie
  (params (Damage :type Float%20%28single-precision%29) (EventInstigator :type Controller%20Object%20Reference) (DamageCauser :type Actor%20Object%20Reference) (bIsTest :type Boolean) (EffectContext :type Gameplay%20Effect%20Context%20Handle%20Structure%20%28by%20ref%29))
  (return false))

(fn ConstructionScript
  (CallFunction|CreateLinkedSeedPodLocationArray)
  (CallFunction|DebugSeedPodManager))
